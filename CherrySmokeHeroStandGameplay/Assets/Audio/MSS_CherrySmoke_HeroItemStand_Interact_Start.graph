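# Mono one-shot MetaSound Source that plays the Cherry Smoke hero item stand interact-start sound.
# It passes its inputs to a MSS_PlayRandomSound_Mono node, which picks from the Sounds array (one wave by default) with slight pitch and volume variation.
# That node's audio, finished trigger and wave duration feed the outputs.
# Playback uses player equip attenuation, the player foley sound class and a concurrency override that blocks excess new instances.
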MetaSoundSource /CherrySmokeHeroStandGameplay/Assets/Audio/MSS_CherrySmoke_HeroItemStand_Interact_Start
inputs [
 MaxPitch {type=Float}
 MaxVolume {type=Float}
 MinPitch {type=Float}
 MinVolume {type=Float}
 Pitch {type=Float}
 Seed {type=Int32}
 Shared State Behavior {type=Enum:SharedStateBehaviorType}
 Sounds {type=WaveAsset:Array}
 UE.Source.OnPlay {type=Trigger}
 Volume {type=Float}
]
interfaces [UE.OutputFormat.Mono:1.0, UE.Source.OneShot:1.0, UE.Source:1.1]
outputs [
 "UE.OutputFormat.Mono.Audio:0" {type=Audio}
 UE.Source.OneShot.OnFinished {type=Trigger}
 WaveDuration {type=Time}
]
pages [Default {
  input_defaults {
   MaxPitch=1.05
   MaxVolume=1.05
   MinPitch=0.95
   MinVolume=0.95
   Pitch=1
   Seed=-1
   Shared State Behavior=1
   UE.Source.OnPlay=null
   Volume=1
   Sounds [&/CherrySmoke_HeroStand_Interact_Start_01]
  }
  nodes [MSS_PlayRandomSound_Mono_0 = &MSS_PlayRandomSound_Mono(
    MaxPitch=^MaxPitch
    MaxVolume=^MaxVolume
    MinPitch=^MinPitch
    MinVolume=^MinVolume
    Pitch=^Pitch
    Seed=^Seed
    Shared State Behavior=^Shared State Behavior
    Sounds=^Sounds
    UE.Source.OnPlay=^UE.Source.OnPlay
    Volume=^Volume
   )]
  links [
   {dst="@UE.OutputFormat.Mono.Audio:0", src=MSS_PlayRandomSound_Mono_0.UE.OutputFormat.Mono.Audio:0}
   {dst="@UE.Source.OneShot.OnFinished", src=MSS_PlayRandomSound_Mono_0.UE.Source.OneShot.OnFinished}
   {dst="@WaveDuration", src=MSS_PlayRandomSound_Mono_0.WaveDuration}
  ]
 }]
properties {
 assetImportData=&self:AssetImportData
 attenuationSettings=/Game/Sounds/Attenuation/Fort_Player_Equip_Generic_Att
 bOverrideConcurrency=true
 maxDistance=1800
 sampleRate=48000
 soundClassObject=&Fort_Player_Foley_Athena
 concurrencyOverrides {maxCount=3, resolutionRule=PreventNew, retriggerTime=0.2}
}
