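# Mono MetaSound source for the looping ambient sound of the CherrySmoke hero stand.
# On play, a single wave player starts the ambient loop wave and feeds the mono output directly.
# It plays quietly through the loot ambient loop sound class with item ambient attenuation.
# Two concurrency sets, one stand-specific and one shared stop-furthest rule, are assigned to limit simultaneous instances.
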MetaSoundSource /CherrySmokeHeroStandGameplay/Assets/Audio/MSS_CherrySmoke_HeroStand_Amb_loop
inputs [UE.Source.OnPlay {type=Trigger}]
interfaces [UE.OutputFormat.Mono:1.0, UE.Source:1.1]
outputs ["UE.OutputFormat.Mono.Audio:0" {type=Audio}]
pages [Default {
  input_defaults {UE.Source.OnPlay=null}
  nodes [WavePlayer_0 = UE.Wave Player:Mono(Play=^UE.Source.OnPlay) {also {defaults {Loop="<MetaSoundBuilderResult.SUCCEEDED: 0>", Wave Asset=&/CherrySmoke_HeroStand_Ambient_lp_01}}}]
  links [{dst="@UE.OutputFormat.Mono.Audio:0", src=WavePlayer_0.Out Mono}]
 }]
properties {
 assetImportData=&self:AssetImportData
 attenuationSettings=/Game/Sounds/Attenuation/Fort_Item_Ambient_Loop_Mono_Default_Att
 bLooping=true
 duration=10000
 maxDistance=1250
 sampleRate=48000
 soundClassObject=/Game/Sounds/FortSoundClassesAndModes/Fort_Loot_Ambient_Loop
 volume=0.15
 concurrencyOverrides {bLimitToOwner=true, maxCount=1, resolutionRule=PreventNew, retriggerTime=0.2}
 concurrencySet [&/CONC_CherrySmoke_HeroStand_Amb, /Game/Sounds/Concurrency_Rules/ThreeMaximum_StopFurthest]
}
