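# Mono one-shot MetaSound source for a CherrySmoke Hero Stand completion sound, routed through the player foley sound class.
# On play, it passes a wave array, currently holding a single Complete clip, to the shared MSS_PlayRandomSound_Mono graph, which supplies the audio and finished trigger.
# It uses generic player-equip attenuation and overrides concurrency with an owner-limited voice cap that prevents new plays when full.
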MetaSoundSource /CherrySmokeHeroStandGameplay/Assets/Audio/MSS_CherrySmoke_HeroStand_Complete
inputs [
 UE.Source.OnPlay {type=Trigger}
 Wave Asset {type=WaveAsset:Array}
]
interfaces [UE.OutputFormat.Mono:1.0, UE.Source.OneShot:1.0, UE.Source:1.1]
outputs [
 "UE.OutputFormat.Mono.Audio:0" {type=Audio}
 UE.Source.OneShot.OnFinished {type=Trigger}
]
pages [Default {
  input_defaults {UE.Source.OnPlay=null, Wave Asset [&/CherrySmoke_HeroStand_Complete_01]}
  nodes [MSS_PlayRandomSound_Mono_0 = &MSS_PlayRandomSound_Mono(UE.Source.OnPlay=^UE.Source.OnPlay, Sounds=^Wave Asset)]
  links [
   {dst="@UE.OutputFormat.Mono.Audio:0", src=MSS_PlayRandomSound_Mono_0.UE.OutputFormat.Mono.Audio:0}
   {dst="@UE.Source.OneShot.OnFinished", src=MSS_PlayRandomSound_Mono_0.UE.Source.OneShot.OnFinished}
  ]
 }]
properties {
 assetImportData=&self:AssetImportData
 attenuationSettings=/Game/Sounds/Attenuation/Fort_Player_Equip_Generic_Att
 bOverrideConcurrency=true
 maxDistance=1800
 sampleRate=48000
 soundClassObject=&Fort_Player_Foley_Athena
 concurrencyOverrides {bLimitToOwner=true, maxCount=4, resolutionRule=PreventNew, retriggerTime=0.2}
}
